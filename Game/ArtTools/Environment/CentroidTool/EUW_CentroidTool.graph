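# Editor utility widget that writes mesh bounds centers into the CentroidOffset property of selected Blueprint assets.
# Its single "Use Bounds for Centroid" button runs an embedded Python script over the Content Browser selection.
# The script loads each Blueprint's class default object, reads its StaticMesh bounds origin, casts it to Vector3f, then saves the asset.
# It skips non-Blueprints and logs warnings for failed loads or null meshes.
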EditorUtilityWidgetBlueprint /Game/ArtTools/Environment/CentroidTool/EUW_CentroidTool
graphs [EventGraph {nodes [
   "On Clicked (Button_Bounds)"() {is=title K2Node_ComponentBoundEvent}
   > Execute Python Script(PythonScript="import unreal\r\n\r\ndef run():\r\n    selected_assets = unreal.EditorUtilityLibrary.get_selected_assets()\r\n\r\n    if not selected_assets:\r\n        unreal.log_warning(\"No assets selected. Please select Blueprint assets in the Content Browser.\")\r\n        return\r\n\r\n    for asset in selected_assets:\r\n        asset_name = asset.get_name()\r\n\r\n        if not isinstance(asset, unreal.Blueprint):\r\n            unreal.log_warning(f\"Skipping '{asset_name}' - not a Blueprint.\")\r\n            continue\r\n\r\n        ## Build the CDO object path: \"{AssetPath}.Default__{AssetName}_C\"\r\n        asset_path = asset.get_path_name().split(\".\")[0]\r\n        cdo_path = f\"{asset_path}.Default__{asset_name}_C\"\r\n\r\n        cdo = unreal.load_object(None, cdo_path)\r\n        if cdo is None:\r\n            unreal.log_warning(f\"Skipping '{asset_name}' - could not load CDO at: {cdo_path}\")\r\n            continue\r\n\r\n        static_mesh = cdo.get_editor_property(\"StaticMesh\")\r\n        if static_mesh is None:\r\n            unreal.log_warning(f\"Skipping '{asset_name}' - no StaticMesh property found.\")\r\n            continue\r\n\r\n        ## get_bounds().origin is the bounding box center in local space (double-precision Vector)\r\n        ## CentroidOffset is Vector3f (single-precision), so we must cast explicitly\r\n        origin = static_mesh.get_bounds().origin\r\n        center = unreal.Vector3f(origin.x, origin.y, origin.z)\r\n\r\n        cdo.set_editor_property(\"CentroidOffset\", center)\r\n\r\n        unreal.EditorAssetLibrary.save_loaded_asset(asset)\r\n\r\n        unreal.log(\r\n            f\"'{asset_name}': CentroidOffset set to \"\r\n            f\"({center.x:.2f}, {center.y:.2f}, {center.z:.2f})\"\r\n        )\r\n\r\nrun()\r\n")
   Event PreConstruct()
   Event Construct()
   Event Tick()
  ]}]
variables [
 NewVar {type=boolean}
 SpawnSphere {type=StaticMeshActor, default=null}
]
widgets {
 info {rootWidgetClass=&HorizontalBox, widgetCount=3}
 widgets [
  HorizontalBox_77 {widgetClassPath=&HorizontalBox, widget=HorizontalBox_77}
  Button_Bounds {
   widgetClassPath=/Script/UIFramework.UIFrameworkCustomButtonWidget
   bIsVariable=true
   parent=HorizontalBox_77
   widget=Button_Bounds
   properties {widgetStyle {
     clickedSlateSound {resourceObject=null}
     disabled {
      bIsDynamicallyLoaded=false
      drawAs=RoundedBox
      imageType=NoImage
      mirroring=NoMirror
      resourceName=None
      resourceObject=null
      tiling=NoTile
      imageSize=<0 0>
      margin=<ltrb 0 0 0 0>
      outlineSettings {
       bUseBrushTransparency=false
       roundingType=FixedRadius
       width=1
       color {colorUseRule=UseColor_ColorTable, specifiedColor=<rgba 0 0 0 0>}
       cornerRadii=<4 4 4 4>
      }
      tintColor {colorUseRule=UseColor_Specified, specifiedColor=<rgba 0.0395 0.0395 0.0395 1>}
      uVRegion {bIsValid=false, max=<0 0>, min=<0 0>}
     }
     disabledForeground {colorUseRule=UseColor_ColorTable, specifiedColor=<rgba 0 0 0 0>}
     hovered {
      bIsDynamicallyLoaded=false
      drawAs=RoundedBox
      imageType=NoImage
      mirroring=NoMirror
      resourceName=None
      resourceObject=null
      tiling=NoTile
      imageSize=<32 32>
      margin=<ltrb 0 0 0 0>
      outlineSettings {
       bUseBrushTransparency=false
       roundingType=FixedRadius
       width=1
       color {colorUseRule=UseColor_ColorTable, specifiedColor=<rgba 0 0 0 0>}
       cornerRadii=<25 25 25 25>
      }
      tintColor {colorUseRule=UseColor_Specified, specifiedColor=<rgba 0.2055 0.599 0.2047 1>}
      uVRegion {bIsValid=false, max=<0 0>, min=<0 0>}
     }
     hoveredForeground {colorUseRule=UseColor_ColorTable, specifiedColor=<rgba 0 0 0 0>}
     hoveredSlateSound {resourceObject=null}
     normal {
      bIsDynamicallyLoaded=false
      drawAs=RoundedBox
      imageType=NoImage
      mirroring=NoMirror
      resourceName=None
      resourceObject=null
      tiling=NoTile
      imageSize=<32 32>
      margin=<ltrb 0 0 0 0>
      outlineSettings {
       bUseBrushTransparency=false
       roundingType=FixedRadius
       width=1
       color {colorUseRule=UseColor_ColorTable, specifiedColor=<rgba 0 0 0 0>}
       cornerRadii=<15 15 15 15>
      }
      tintColor {colorUseRule=UseColor_Specified, specifiedColor=<rgba 0.0395 0.0395 0.0395 1>}
      uVRegion {bIsValid=false, max=<0 0>, min=<0 0>}
     }
     normalForeground {colorUseRule=UseColor_ColorTable, specifiedColor=<rgba 0 0 0 0>}
     normalPadding=<ltrb 0 0 0 0>
     pressed {
      bIsDynamicallyLoaded=false
      drawAs=RoundedBox
      imageType=NoImage
      mirroring=NoMirror
      resourceName=None
      resourceObject=null
      tiling=NoTile
      imageSize=<32 32>
      margin=<ltrb 0 0 0 0>
      outlineSettings {
       bUseBrushTransparency=false
       roundingType=FixedRadius
       width=1
       color {colorUseRule=UseColor_ColorTable, specifiedColor=<rgba 0 0 0 0>}
       cornerRadii=<4 4 4 4>
      }
      tintColor {colorUseRule=UseColor_Specified, specifiedColor=<rgba 0.0284 0.0284 0.0284 1>}
      uVRegion {bIsValid=false, max=<0 0>, min=<0 0>}
     }
     pressedForeground {colorUseRule=UseColor_ColorTable, specifiedColor=<rgba 0 0 0 0>}
     pressedPadding=<ltrb 0 0 0 0>
     pressedSlateSound {resourceObject=null}
    }}
   slot {class=HorizontalBoxSlot, properties {padding=<ltrb 10 10 10 10>, size {sizeRule=Fill, value=1}}}
  }
  TextBlock_1 {
   widgetClassPath=&TextBlock
   parent=Button_Bounds
   widget=TextBlock_1
   properties {
    text=Use Bounds for Centroid
    font {
     bForceMonospaced=false
     bMaterialIsStencil=false
     fontMaterial=null
     fontObject=&Roboto
     letterSpacing=0
     monospacedWidth=1
     size=18
     skewAmount=0
     typefaceFontName=Bold
     outlineSettings {
      bApplyOutlineToDropShadows=false
      bMiteredCorners=false
      bSeparateFillAlpha=false
      outlineMaterial=null
      outlineSize=0
      outlineColor=<rgba 0 0 0 1>
     }
    }
   }
   slot {class=ButtonSlot, properties {padding=<ltrb 10 10 10 10>}}
  }
 ]
}
